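# Physics asset for the Fortnite male Cavern generic cape accessory, defining collision bodies and joints for its dynamic cape and belt bones.
# Center, left and right seven-segment chains hang from master bodies joined to backpack_root with free swing and twist, using boxes near the top and capsules below.
# Cross-constraints tie both side chains to the center at segments 05 and 07.
# Three-link belt chains branch from each side's first segment with zero twist limits.
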PhysicsAsset /BRCosmetics/Accessories/FORT_Capes/Cape_M_MED_Cavern_Generic/Meshes/Fortnite_CavernMale_Dyn
bodies {
 backpack_root [{shapeType=Capsule, length=28.1261, radius=23.9062, center=<-0.0001 24.0106 -25.4804>, rotation=<rot 0 0 5.018>}]
 belt_00_l [{shapeType=Capsule, length=1, radius=3, center=<2.5 -3 0>, rotation=<rot 90 0 0>}]
 belt_00_r [{shapeType=Capsule, length=1, radius=3, center=<2.5 -3 0>, rotation=<rot 90 0 0>}]
 belt_01_l [{shapeType=Capsule, length=1, radius=3, center=<2.5 -3 0>, rotation=<rot 90 0 0>}]
 belt_01_r [{shapeType=Capsule, length=1, radius=3, center=<2.5 -3 0>, rotation=<rot 90 0 0>}]
 belt_02_l [{shapeType=Capsule, length=2.5, radius=3.5, center=<3.3131 -3 0>, rotation=<rot 90 0 0>}]
 belt_02_r [{shapeType=Capsule, length=2.5, radius=3.5, center=<3.3131 -3 0>, rotation=<rot 90 0 0>}]
 dyn_cape_01 [{shapeType=Box, extentX=19.5, extentY=11.5, extentZ=1.5, shapeName=dyn_cape_01_box, center=<9 0 0.8894>}]
 dyn_cape_02 [{shapeType=Capsule, length=9.602, radius=8.1308, shapeName=dyn_cape_01_capsule, center=<7.15 0 5>, rotation=<rot 90 0 0>}]
 dyn_cape_03 [{shapeType=Capsule, length=9.602, radius=8.1308, shapeName=dyn_cape_01_capsule, center=<7.15 0 2>, rotation=<rot 90 0 0>}]
 dyn_cape_04 [{shapeType=Capsule, length=9.602, radius=8.1308, shapeName=dyn_cape_01_capsule, center=<7.15 0 2>, rotation=<rot 90 0 0>}]
 dyn_cape_05 [{shapeType=Capsule, length=9.602, radius=8.1308, shapeName=dyn_cape_01_capsule, center=<7.15 0 2>, rotation=<rot 90 0 0>}]
 dyn_cape_06 [{shapeType=Capsule, length=9.602, radius=8.1308, shapeName=dyn_cape_01_capsule, center=<7.15 0 2>, rotation=<rot 90 0 0>}]
 dyn_cape_07 [{shapeType=Capsule, length=1.2243, radius=8.1308, shapeName=dyn_cape_01_capsule, center=<5.8445 0 2>, rotation=<rot 90 0 0>}]
 dyn_cape_MSTR [{shapeType=Capsule, length=0.505, radius=0.505, shapeName=dyn_cape_r_MSTR_capsule, rotation=<rot 0 -90 0>}]
 dyn_cape_l_01 [{shapeType=Box, extentX=19.5, extentY=11.5, extentZ=1.5, shapeName=dyn_cape_l_01_box, center=<9 -5 0>}]
 dyn_cape_l_02 [{shapeType=Box, extentX=17.5, extentY=12.5, extentZ=3.5, shapeName=dyn_cape_l_02_box, center=<9 -2.5 0>}]
 dyn_cape_l_03 [{shapeType=Capsule, length=9.602, radius=8.1308, shapeName=dyn_cape_01_capsule, center=<7.15 0 2>, rotation=<rot 90 0 0>}]
 dyn_cape_l_04 [{shapeType=Capsule, length=9.602, radius=8.1308, shapeName=dyn_cape_01_capsule, center=<7.15 0 2>, rotation=<rot 90 0 0>}]
 dyn_cape_l_05 [{shapeType=Capsule, length=9.602, radius=8.1308, shapeName=dyn_cape_01_capsule, center=<7.15 0 2>, rotation=<rot 90 0 0>}]
 dyn_cape_l_06 [{shapeType=Capsule, length=9.602, radius=8.1308, shapeName=dyn_cape_01_capsule, center=<7.15 0 2>, rotation=<rot 90 0 0>}]
 dyn_cape_l_07 [{shapeType=Capsule, length=1.2243, radius=8.1308, shapeName=dyn_cape_01_capsule, center=<5.8445 0 2>, rotation=<rot 90 0 0>}]
 dyn_cape_l_MSTR [{shapeType=Capsule, length=0.505, radius=0.505, shapeName=dyn_cape_r_MSTR_capsule, rotation=<rot 0 -90 0>}]
 dyn_cape_r_01 [{shapeType=Box, extentX=19.5, extentY=11.5, extentZ=1.5, shapeName=dyn_cape_r_01_box, center=<9 5 0>}]
 dyn_cape_r_02 [{shapeType=Box, extentX=17.5, extentY=12.5, extentZ=3.5, shapeName=dyn_cape_r_02_box, center=<9 2.5 0>}]
 dyn_cape_r_03 [{shapeType=Capsule, length=9.602, radius=8.1308, shapeName=dyn_cape_01_capsule, center=<7.15 0 2>, rotation=<rot 90 0 0>}]
 dyn_cape_r_04 [{shapeType=Capsule, length=9.602, radius=8.1308, shapeName=dyn_cape_01_capsule, center=<7.15 0 2>, rotation=<rot 90 0 0>}]
 dyn_cape_r_05 [{shapeType=Capsule, length=9.602, radius=8.1308, shapeName=dyn_cape_01_capsule, center=<7.15 0 2>, rotation=<rot 90 0 0>}]
 dyn_cape_r_06 [{shapeType=Capsule, length=9.602, radius=8.1308, shapeName=dyn_cape_01_capsule, center=<7.15 0 2>, rotation=<rot 90 0 0>}]
 dyn_cape_r_07 [{shapeType=Capsule, length=1.2243, radius=8.1308, shapeName=dyn_cape_01_capsule, center=<5.8445 0 2>, rotation=<rot 90 0 0>}]
 dyn_cape_r_MSTR [{shapeType=Capsule, length=0.505, radius=0.505, shapeName=dyn_cape_r_MSTR_capsule, rotation=<rot 0 -90 0>}]
}
constraints [
 {bone1Name=dyn_cape_r_01, bone2Name=dyn_cape_r_MSTR, swing1LimitDegrees=10, swing2LimitDegrees=45, twistLimitDegrees=5}
 {bone1Name=dyn_cape_r_02, bone2Name=dyn_cape_r_01, swing1LimitDegrees=20, swing2LimitDegrees=45, twistLimitDegrees=5}
 {bone1Name=dyn_cape_r_03, bone2Name=dyn_cape_r_02, swing1LimitDegrees=25, swing2LimitDegrees=45, twistLimitDegrees=2}
 {bone1Name=dyn_cape_r_04, bone2Name=dyn_cape_r_03, swing1LimitDegrees=25, swing2LimitDegrees=45, twistLimitDegrees=2}
 {bone1Name=dyn_cape_r_05, bone2Name=dyn_cape_r_04, swing1LimitDegrees=25, swing2LimitDegrees=45, twistLimitDegrees=2}
 {bone1Name=dyn_cape_r_06, bone2Name=dyn_cape_r_05, swing1LimitDegrees=25, swing2LimitDegrees=45, twistLimitDegrees=2}
 {bone1Name=dyn_cape_r_07, bone2Name=dyn_cape_r_06, swing1LimitDegrees=25, swing2LimitDegrees=45, twistLimitDegrees=2}
 {bone1Name=dyn_cape_l_01, bone2Name=dyn_cape_l_MSTR, swing1LimitDegrees=10, swing2LimitDegrees=45, twistLimitDegrees=5}
 {bone1Name=dyn_cape_l_02, bone2Name=dyn_cape_l_01, swing1LimitDegrees=20, swing2LimitDegrees=45, twistLimitDegrees=5}
 {bone1Name=dyn_cape_l_03, bone2Name=dyn_cape_l_02, swing1LimitDegrees=25, swing2LimitDegrees=45, twistLimitDegrees=2}
 {bone1Name=dyn_cape_l_04, bone2Name=dyn_cape_l_03, swing1LimitDegrees=25, swing2LimitDegrees=45, twistLimitDegrees=2}
 {bone1Name=dyn_cape_l_05, bone2Name=dyn_cape_l_04, swing1LimitDegrees=25, swing2LimitDegrees=45, twistLimitDegrees=2}
 {bone1Name=dyn_cape_l_06, bone2Name=dyn_cape_l_05, swing1LimitDegrees=25, swing2LimitDegrees=45, twistLimitDegrees=2}
 {bone1Name=dyn_cape_l_07, bone2Name=dyn_cape_l_06, swing1LimitDegrees=25, swing2LimitDegrees=45, twistLimitDegrees=2}
 {bone1Name=dyn_cape_01, bone2Name=dyn_cape_MSTR, swing1LimitDegrees=25, swing2LimitDegrees=45, twistLimitDegrees=5}
 {bone1Name=dyn_cape_02, bone2Name=dyn_cape_01, swing1LimitDegrees=25, swing2LimitDegrees=45, twistLimitDegrees=5}
 {bone1Name=dyn_cape_03, bone2Name=dyn_cape_02, swing1LimitDegrees=25, swing2LimitDegrees=45, twistLimitDegrees=2}
 {bone1Name=dyn_cape_04, bone2Name=dyn_cape_03, swing1LimitDegrees=25, swing2LimitDegrees=45, twistLimitDegrees=2}
 {bone1Name=dyn_cape_05, bone2Name=dyn_cape_04, swing1LimitDegrees=25, swing2LimitDegrees=45, twistLimitDegrees=2}
 {bone1Name=dyn_cape_06, bone2Name=dyn_cape_05, swing1LimitDegrees=25, swing2LimitDegrees=45, twistLimitDegrees=2}
 {bone1Name=dyn_cape_07, bone2Name=dyn_cape_06, swing1LimitDegrees=25, swing2LimitDegrees=45, twistLimitDegrees=2}
 {bone1Name=dyn_cape_r_05, bone2Name=dyn_cape_05, swing1LimitDegrees=25, swing2LimitDegrees=45, twistLimitDegrees=2}
 {bone1Name=dyn_cape_l_05, bone2Name=dyn_cape_05, swing1LimitDegrees=25, swing2LimitDegrees=45, twistLimitDegrees=2}
 {bone1Name=dyn_cape_l_07, bone2Name=dyn_cape_07, swing1LimitDegrees=25, swing2LimitDegrees=45, twistLimitDegrees=2}
 {bone1Name=dyn_cape_r_07, bone2Name=dyn_cape_07, swing1LimitDegrees=25, swing2LimitDegrees=45, twistLimitDegrees=2}
 {bone1Name=belt_00_l, bone2Name=dyn_cape_l_01, swing1LimitDegrees=8, swing2LimitDegrees=8, twistLimitDegrees=0}
 {bone1Name=belt_01_l, bone2Name=belt_00_l, swing1LimitDegrees=20, swing2LimitDegrees=20, twistLimitDegrees=0}
 {bone1Name=belt_02_l, bone2Name=belt_01_l, swing1LimitDegrees=20, swing2LimitDegrees=20, twistLimitDegrees=0}
 {bone1Name=belt_01_r, bone2Name=belt_00_r, swing1LimitDegrees=20, swing2LimitDegrees=20, twistLimitDegrees=0}
 {bone1Name=belt_00_r, bone2Name=dyn_cape_r_01, swing1LimitDegrees=8, swing2LimitDegrees=8, twistLimitDegrees=0}
 {bone1Name=belt_02_r, bone2Name=belt_01_r, swing1LimitDegrees=20, swing2LimitDegrees=20, twistLimitDegrees=0}
 {
  bone1Name=dyn_cape_r_MSTR
  bone2Name=backpack_root
  swing1LimitDegrees=45
  swing1Motion=Free
  swing2LimitDegrees=45
  swing2Motion=Free
  twistLimitDegrees=45
  twistMotion=Free
 }
 {
  bone1Name=dyn_cape_l_MSTR
  bone2Name=backpack_root
  swing1LimitDegrees=45
  swing1Motion=Free
  swing2LimitDegrees=45
  swing2Motion=Free
  twistLimitDegrees=45
  twistMotion=Free
 }
 {
  bone1Name=dyn_cape_MSTR
  bone2Name=backpack_root
  swing1LimitDegrees=45
  swing1Motion=Free
  swing2LimitDegrees=45
  swing2Motion=Free
  twistLimitDegrees=45
  twistMotion=Free
 }
]
